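; Dance-off quest actor where two sides score by emoting on randomly chosen panels of two nine-panel floors.
; Both sides being occupied starts timed server rounds that bind panel overlaps, though only side A's panel gets lit and raised.
; A looping timer credits emoting occupants through an external scoreboard, which ends the game after max rounds.
; RepNotifies drive attract flashing, music, point cues and win/lose effects through dynamic floor materials.
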(event Custom|WinFX_Event
  (bind _win_system_a (Variables|Default|GetWin%20System_A))
  (bind _lose_system_b (Variables|Default|GetLose%20System_B))
  (bind _score_side_a (Variables|Default|GetScore_Side_A))
  (bind _score_side_b (Variables|Default|GetScore_Side_B))
  (bind _win_system_a_2 (Variables|Default|GetWin%20System_A))
  (bind _win_system_b (Variables|Default|GetWin%20System_B))
  (bind _danceoff_defeat_sound (Variables|Default|GetDanceOff_Defeat_Sound))
  (bind _danceoff_victory_sound (Variables|Default|GetDanceOff_Victory_Sound))
  ;[0.8 0.4225 0.06506] WIN / LOSE
  ;;[0.3062 0.5219 1] Play particle effects on-win
  (if (== _score_side_a _score_side_b)
    (if (and (== _score_side_a 0) (== _score_side_b 0))
      (Components|Activation|Activate _lose_system_b false) ; TIELOSS
      ;;[1 0 0.8711] AUDIO - WIN / LOSE BEHAVIOR HERE
      (Audio|PlaySoundatLocation _danceoff_defeat_sound (Transformation|GetWorldLocation _lose_system_b))
      (Audio|PlaySoundatLocation _danceoff_defeat_sound (Transformation|GetWorldLocation (Variables|Default|GetLose%20System_A)))
      ;;[0.3062 0.5219 1] Set Floor Values to Win or Lose Condition Look
      (for _array_element (Variables|Default|GetFloors_A)
        (Rendering|Material|SetScalarParameterValue _array_element "Value multiplier" 1.0 :!class /Script/Engine.MaterialInstanceDynamic)
        (Rendering|Material|SetScalarParameterValue _array_element "Fail" 1.0 :!class /Script/Engine.MaterialInstanceDynamic)
        (Rendering|Material|SetScalarParameterValue _array_element "FullPanelGlow" 1.0 :!class /Script/Engine.MaterialInstanceDynamic))
      (for _array_element_2 (Variables|Default|GetFloors_B)
        (Rendering|Material|SetScalarParameterValue _array_element_2 "Value multiplier" 1.0 :!class /Script/Engine.MaterialInstanceDynamic)
        (Rendering|Material|SetScalarParameterValue _array_element_2 "Fail" 1.0 :!class /Script/Engine.MaterialInstanceDynamic)
        (Rendering|Material|SetScalarParameterValue _array_element_2 "FullPanelGlow" 1.0 :!class /Script/Engine.MaterialInstanceDynamic))
      (else
        (Components|Activation|Activate _win_system_a false) ; TIE WIN
        (Audio|PlaySoundatLocation _danceoff_victory_sound (Transformation|GetWorldLocation _win_system_a))
        (Audio|PlaySoundatLocation _danceoff_victory_sound (Transformation|GetWorldLocation (Variables|Default|GetWin%20System_B)))
        (for _array_element_3 (Variables|Default|GetFloors_A)
          (Rendering|Material|SetScalarParameterValue _array_element_3 "Value multiplier" 1.0 :!class /Script/Engine.MaterialInstanceDynamic)
          (Rendering|Material|SetScalarParameterValue _array_element_3 "FullPanelGlow" 1.0 :!class /Script/Engine.MaterialInstanceDynamic))
        (for _array_element_4 (Variables|Default|GetFloors_B)
          (Rendering|Material|SetScalarParameterValue _array_element_4 "Value multiplier" 1.0 :!class /Script/Engine.MaterialInstanceDynamic)
          (Rendering|Material|SetScalarParameterValue _array_element_4 "FullPanelGlow" 1.0 :!class /Script/Engine.MaterialInstanceDynamic))))
    (elif (> _score_side_a _score_side_b)
      (Development|PrintString "SIDE A WINS" true true "(R=1.000000,G=0.527330,B=0.000000,A=1.000000)" 5.0)
      (Components|Activation|Activate _win_system_a_2 false) ; SIDE A WINS
      (Audio|PlaySoundatLocation (Variables|Default|GetDanceOff_Victory_Sound) (Transformation|GetWorldLocation _win_system_a_2))
      (Audio|PlaySoundatLocation (Variables|Default|GetDanceOff_Defeat_Sound) (Transformation|GetWorldLocation (Variables|Default|GetLose%20System_B)))
      (for _array_element_5 (Variables|Default|GetFloors_A)
        (Rendering|Material|SetScalarParameterValue _array_element_5 "Value multiplier" 1.0 :!class /Script/Engine.MaterialInstanceDynamic)
        (Rendering|Material|SetScalarParameterValue _array_element_5 "FullPanelGlow" 1.0 :!class /Script/Engine.MaterialInstanceDynamic))
      (for _array_element_6 (Variables|Default|GetFloors_B)
        (Rendering|Material|SetScalarParameterValue _array_element_6 "Value multiplier" 1.0 :!class /Script/Engine.MaterialInstanceDynamic)
        (Rendering|Material|SetScalarParameterValue _array_element_6 "Fail" 1.0 :!class /Script/Engine.MaterialInstanceDynamic)
        (Rendering|Material|SetScalarParameterValue _array_element_6 "FullPanelGlow" 1.0 :!class /Script/Engine.MaterialInstanceDynamic))
      (else
        (Development|PrintString "SICE B WINS" true true "(R=0.155076,G=1.000000,B=0.000000,A=1.000000)" 5.0)
        (Components|Activation|Activate _win_system_b false) ; SIDE B WINS
        (Audio|PlaySoundatLocation (Variables|Default|GetDanceOff_Victory_Sound) (Transformation|GetWorldLocation _win_system_b))
        (Audio|PlaySoundatLocation (Variables|Default|GetDanceOff_Defeat_Sound) (Transformation|GetWorldLocation (Variables|Default|GetLose%20System_A)))
        (for _array_element_7 (Variables|Default|GetFloors_A)
          (Rendering|Material|SetScalarParameterValue _array_element_7 "Value multiplier" 1.0 :!class /Script/Engine.MaterialInstanceDynamic)
          (Rendering|Material|SetScalarParameterValue _array_element_7 "Fail" 1.0 :!class /Script/Engine.MaterialInstanceDynamic)
          (Rendering|Material|SetScalarParameterValue _array_element_7 "FullPanelGlow" 1.0 :!class /Script/Engine.MaterialInstanceDynamic))
        (for _array_element_8 (Variables|Default|GetFloors_B)
          (Rendering|Material|SetScalarParameterValue _array_element_8 "Value multiplier" 1.0 :!class /Script/Engine.MaterialInstanceDynamic)
          (Rendering|Material|SetScalarParameterValue _array_element_8 "FullPanelGlow" 1.0 :!class /Script/Engine.MaterialInstanceDynamic)))))
  (CallFunction|ClearScore)
  (Audio|Components|Audio|FadeOut (Variables|Default|Getdisco_puzzle_danceoff_music) 0.0 0.0 "Linear" :!class /Script/Engine.AudioComponent)
  ;;[0.3062 0.5219 1] Resets all Color Values on Floor from Win/Lose State
  (Utilities|FlowControl|Delay 2.0) ; Let Floor Look Linger a moment
  (for _array_element_9 (Variables|Default|GetFloors_A)
    (Rendering|Material|SetScalarParameterValue _array_element_9 "Value multiplier" 0.0 :!class /Script/Engine.MaterialInstanceDynamic)
    (Rendering|Material|SetScalarParameterValue _array_element_9 "Fail" 0.0 :!class /Script/Engine.MaterialInstanceDynamic)
    (Rendering|Material|SetScalarParameterValue _array_element_9 "FullPanelGlow" 0.0 :!class /Script/Engine.MaterialInstanceDynamic))
  (for _array_element_10 (Variables|Default|GetFloors_B)
    (Rendering|Material|SetScalarParameterValue _array_element_10 "Value multiplier" 0.0 :!class /Script/Engine.MaterialInstanceDynamic)
    (Rendering|Material|SetScalarParameterValue _array_element_10 "Fail" 0.0 :!class /Script/Engine.MaterialInstanceDynamic)
    (Rendering|Material|SetScalarParameterValue _array_element_10 "FullPanelGlow" 0.0 :!class /Script/Engine.MaterialInstanceDynamic)))

(event EventBeginPlay
  ;[1 1 1] BEGIN PLAY AND GAME MUSIC
  (Transformation|SetRelativeLocation (Variables|Default|GetOverlaps_A) "0, 0,-192.000000" false false)
  (Transformation|SetRelativeLocation (Variables|Default|GetOverlaps_B) "0, 0,-192.000000" false false))

(event Custom|GameMusic
  ;[1 1 1] BEGIN PLAY AND GAME MUSIC
  (Audio|Components|Audio|Play (Variables|Default|Getdisco_puzzle_danceoff_music) 0.0))

(event Custom|FlashPanel_A
  (params (PanelMaterial :type Material%20Instance%20Dynamic%20Object%20Reference))
  ;[0 1 0.8448] Attract Mode
  (Utilities|Time|SetTimerbyEvent (delegate Custom|FlashAOff) 0.15)
  (Rendering|Material|SetScalarParameterValue PanelMaterial "Value multiplier" 1.0 :!class /Script/Engine.MaterialInstanceDynamic)
  (Rendering|Material|SetScalarParameterValue PanelMaterial "FullPanelGlow" 2.0 :!class /Script/Engine.MaterialInstanceDynamic))

(event Custom|Attract_A
  ;[0 1 0.8448] Attract Mode
  (CallFunction|FlashPanelA (Utilities|Array|Get(acopy) (Variables|Default|GetFloors_B) (Math|Random|RandomInteger (Utilities|Array|Length (Variables|Default|GetFloors_B)))))
  (Utilities|Time|SetTimerbyFunctionName 0 "Attract_A" 0.3))

(event Custom|FlashAOff
  ;[0 1 0.8448] Attract Mode
  (Rendering|Material|SetScalarParameterValue (stale Custom|FlashPanel_A PanelMaterial) "Value multiplier" 0.0 :!class /Script/Engine.MaterialInstanceDynamic)
  (Rendering|Material|SetScalarParameterValue (stale Custom|FlashPanel_A PanelMaterial) "FullPanelGlow" 0.0 :!class /Script/Engine.MaterialInstanceDynamic))

(event Custom|Attract_B
  ;[0 1 0.8448] Attract Mode
  (CallFunction|FlashPanelB (Utilities|Array|Get(acopy) (Variables|Default|GetFloors_A) (Math|Random|RandomInteger (Utilities|Array|Length (Variables|Default|GetFloors_A)))))
  (Utilities|Time|SetTimerbyFunctionName 0 "Attract_B" 0.3))

(event Custom|FlashPanel_B
  (params (PanelMaterial :type Material%20Instance%20Dynamic%20Object%20Reference))
  ;[0 1 0.8448] Attract Mode
  (Utilities|Time|SetTimerbyEvent (delegate Custom|FlashBOff) 0.15)
  (Rendering|Material|SetScalarParameterValue PanelMaterial "Value multiplier" 1.0 :!class /Script/Engine.MaterialInstanceDynamic)
  (Rendering|Material|SetScalarParameterValue PanelMaterial "FullPanelGlow" 1.0 :!class /Script/Engine.MaterialInstanceDynamic))

(event Custom|RoundTimer
  ;[1 1 1] Round work
  (Networking|SwitchHasAuthority
    (:Authority
      (Variables|Default|SetbScoreReady false)
      (Utilities|Array|Clear (Variables|Default|GetPlayers_A))
      (Utilities|Array|Clear (Variables|Default|GetPlayers_B))
      (CallFunction|PanelsShutdown)
      (if (>= (Variables|Default|GetNumRounds) (Variables|Default|GetMaxRounds))
        (Variables|Default|SetbPanelActive false)
        (Class|TwinScoreboardINF|EndGame(Message) (Variables|Default|GetScoreboard))
        (else
          (Utilities|FlowControl|Delay 0.05) ; Delay before next round starts
          (CallFunction|BeginRound)
          (Utilities|Time|SetTimerbyFunctionName 0 "RoundTimer" (Variables|Default|GetRoundLength) true)
          (Development|PrintString "BEGIN ROUND!" true true "(R=0.355953,G=1.000000,B=0.000000,A=1.000000)"))))))

(event Custom|FlashBOff
  ;[0 1 0.8448] Attract Mode
  (Rendering|Material|SetScalarParameterValue (stale Custom|FlashPanel_B PanelMaterial) "Value multiplier" 0.0 :!class /Script/Engine.MaterialInstanceDynamic)
  (Rendering|Material|SetScalarParameterValue (stale Custom|FlashPanel_B PanelMaterial) "FullPanelGlow" 0.0 :!class /Script/Engine.MaterialInstanceDynamic))

(event Custom|ResumeMainOverlap
  (bind _overlapsidea (Variables|Default|GetOverlapSideA))
  ;[0.123 0 1] Main overlap work, make sure its out of the way when not being used.
  (Transformation|SetRelativeLocation _overlapsidea "0, 0, 0" false false)
  (Components|Activation|SetActive _overlapsidea true false))

(event Custom|PauseMainOverlap
  (bind _overlapsidea (Variables|Default|GetOverlapSideA))
  ;[0.123 0 1] Main overlap work, make sure its out of the way when not being used.
  (Transformation|SetRelativeLocation _overlapsidea "0, 0,-500.000000" false false)
  (Components|Activation|SetActive _overlapsidea false false))

(event Custom|BeginRound
  ;[1 0.3126 0] Pick panels to light up - Server
  (if (Variables|Default|GetbPanelActive)
    (Utilities|FlowControl|Sequence
      (:then_0
        (Networking|SwitchHasAuthority
          (:Authority
            (Variables|Default|SetA_Index (Math|Random|RandomIntegerinRange 0 8))
            (Variables|Default|SetB_Index (Math|Random|RandomIntegerinRange 0 8)))))
      (:then_1
        (CallFunction|LightUp)
        (Variables|Default|SetbScoreReady true)
        (Math|Integer|IncrementInt (Variables|Default|GetNumRounds))))))

(event Custom|ClearGame
  ;[0.01313 0 0.5104] Game Init / Reinit
  (Variables|Default|SetbScoreReady false)
  (Utilities|FlowControl|Sequence
    (:then_0
      (Utilities|Array|Clear (Variables|Default|GetPlayers_A))
      (Utilities|Array|Clear (Variables|Default|GetPlayers_B))
      (Variables|Default|SetScore_Side_B 0)
      (Variables|Default|SetScore_Side_A 0))
    (:then_1
      (Utilities|Time|ClearTimerbyFunctionName 0 "ScoringLoop")
      (Utilities|Time|ClearTimerbyFunctionName 0 "BeginRound")
      (Utilities|Time|ClearTimerbyFunctionName 0 "RoundTimer")
      (Variables|Default|SetNumRounds 0)
      (Utilities|FlowControl|Delay 1.0)
      (CallFunction|ResumeMainOverlap))))

(event Custom|LightUp
  (bind _a_index (Variables|Default|GetA_Index))
  (bind _overlaps_a (Variables|Default|GetOverlaps_A))
  (bind _output (Utilities|Array|Get(acopy) _overlaps_a _a_index))
  (bind _b_index (Variables|Default|GetB_Index))
  (bind _overlaps_b (Variables|Default|GetOverlaps_B))
  (bind _output_2 (Utilities|Array|Get(acopy) _overlaps_b _b_index))
  ;[0.2969 0.2969 0.2969] Light up Panels and bind to overlaps
  (Utilities|FlowControl|Delay 0.1)
  (Networking|SwitchHasAuthority
    (:Authority
      (CallFunction|PanelOn)
      ;;[1 0.3126 0] Event Binds
      (Collision|BindEventtoOnComponentBeginOverlap _output (EventDispatchers|CreateEvent :!event Custom|PanelCheck_A))
      (Collision|BindEventtoOnComponentEndOverlap _output (EventDispatchers|CreateEvent :!event Custom|PanelLeave_A))
      (Collision|BindEventtoOnComponentBeginOverlap _output_2 (EventDispatchers|CreateEvent :!event Custom|PanelCheck_B))
      (Collision|BindEventtoOnComponentEndOverlap _output_2 (EventDispatchers|CreateEvent :!event Custom|PanelLeave_B))
      (Transformation|SetRelativeLocation (Utilities|Array|Get(acopy) (Variables|Default|GetOverlaps_A) (Variables|Default|GetA_Index)) "0, 0, 0" false false))))

(event Custom|PanelOn
  (bind _panelhighlight_sound (Variables|Default|GetPanelHighlight_Sound))
  (bind _highlightedvfx_a (Variables|Default|GetHighlightedVFX_A))
  ;[0 1 0.02272] Activate Lights
  (Rendering|Material|SetScalarParameterValue (Utilities|Array|Get(acopy) (Variables|Default|GetFloors_A) (Variables|Default|GetA_Index)) "Value multiplier" 3.0 :!class /Script/Engine.MaterialInstanceDynamic)
  ;;[0 0.7225 1] Particle Selection effect should only run on Console and PC
  (if (not (Fortnite|IsMobilePlatform))
    (bind _returnvalue (CallFunction|HighlightVFXLocationSideA))
    (Transformation|SetRelativeLocation (Variables|Default|GetHighlightedVFX_A) _returnvalue false false)
    (bind _returnvalue_2 (CallFunction|HighlightVFXLocationSideB))
    (Transformation|SetRelativeLocation (Variables|Default|GetHighlightedVFX_B) _returnvalue_2 false false)
    (Components|Activation|Activate _highlightedvfx_a false))
  ;[1 0 0.8711] AUDIO - PANEL ON/OFF HERE
  (Audio|PlaySoundatLocation _panelhighlight_sound (Transformation|GetWorldLocation _highlightedvfx_a))
  (Audio|PlaySoundatLocation _panelhighlight_sound (Transformation|GetWorldLocation (Variables|Default|GetHighlightedVFX_B))))

(event Custom|PanelCheck_A
  (params (OverlappedComponent :type Primitive%20Component%20Object%20Reference) (OtherActor :type Actor%20Object%20Reference) (OtherComp :type Primitive%20Component%20Object%20Reference) (OtherBodyIndex :type Integer) (bFromSweep :type Boolean) (SweepResult :type Hit%20Result%20Structure%20%28by%20ref%29))
  ;[0.9093 0.3501 1] Panel Occupancy A
  (if (and (Variables|Default|GetbPanelActive) (Variables|Default|GetbScoreReady))
    (bind _asfort_player_pawn_athena (Utilities|Casting|CastToFortPlayerPawnAthena OtherActor))
    (bind _asathena_player_controller (Utilities|Casting|CastToAthena_PlayerController (Pawn|GetController _asfort_player_pawn_athena)))
    (Utilities|Array|AddUnique (Variables|Default|GetPlayers_A) _asathena_player_controller)))

(event Custom|ScoringLoop
  (bind _scoreboard (Variables|Default|GetScoreboard))
  (bind _scoringplayers_a (Variables|Default|GetScoringPlayers_A))
  (bind _scoringplayers_b (Variables|Default|GetScoringPlayers_B))
  ;[1 0.25 0.25] Scoring Behavior
  (Networking|SwitchHasAuthority
    (:Authority
      (if (and (Variables|Default|GetbPanelActive) (Variables|Default|GetbScoreReady))
        (Utilities|FlowControl|Sequence
          (:then_0
            (bind _array_element (Utilities|Array|ForEachLoopwithBreak (Variables|Default|GetPlayers_A))
              (:LoopBody
                (if (Emote|IsPlayingEmote _array_element)
                  (Utilities|FlowControl|Sequence
                    (:then_0
                      _)
                    (:then_1
                      _))
                  (for _array_element_2 (Variables|Default|GetPlayers_A)
                    (if (Emote|IsPlayingEmote _array_element_2)
                      ;;[1 1 1] Create list of scoring players
                      (Utilities|Array|AddUnique (Variables|Default|GetScoringPlayers_A) _array_element_2)))
                  (Class|BPTwinScoreBoardPlace|MoveScore _scoreboard 1 true _scoringplayers_a _scoringplayers_b)
                  (CallFunction|ScoreA)))))
          (:then_1
            (bind _array_element_3 (Utilities|Array|ForEachLoopwithBreak (Variables|Default|GetPlayers_B))
              (:LoopBody
                (if (Emote|IsPlayingEmote _array_element_3)
                  (Utilities|FlowControl|Sequence
                    (:then_0
                      _)
                    (:then_1
                      _))
                  (for _array_element_4 (Variables|Default|GetPlayers_B)
                    (if (Emote|IsPlayingEmote _array_element_4)
                      (Utilities|Array|AddUnique (Variables|Default|GetScoringPlayers_B) _array_element_4)))
                  (Class|BPTwinScoreBoardPlace|MoveScore _scoreboard 1 false _scoringplayers_a _scoringplayers_b)
                  (CallFunction|ScoreB))))))))))

(event Custom|PanelsShutdown
  ;[0 1 0.02272] Activate Lights
  (Transformation|SetRelativeLocation (Utilities|Array|Get(acopy) (Variables|Default|GetOverlaps_A) (Variables|Default|GetA_Index)) "0, 0,-192.000000" false false)
  (Rendering|Material|SetScalarParameterValue (Utilities|Array|Get(acopy) (Variables|Default|GetFloors_A) (Variables|Default|GetA_Index)) "Value multiplier" 0.0 :!class /Script/Engine.MaterialInstanceDynamic)
  (if (not (Fortnite|IsMobilePlatform))
    (Components|Activation|Deactivate (Variables|Default|GetHighlightedVFX_A))))

(event Custom|PanelLeave_A
  (params (OverlappedComponent :type Primitive%20Component%20Object%20Reference) (OtherActor :type Actor%20Object%20Reference) (OtherComp :type Primitive%20Component%20Object%20Reference) (OtherBodyIndex :type Integer))
  ;[0.3168 0 0.4688] Panel Departure A
  (if (and (Variables|Default|GetbPanelActive) (Variables|Default|GetbScoreReady))
    (bind _asfort_player_pawn_athena (Utilities|Casting|CastToFortPlayerPawnAthena OtherActor))
    (bind _asathena_player_controller (Utilities|Casting|CastToAthena_PlayerController (Pawn|GetController _asfort_player_pawn_athena)))
    (Utilities|Array|RemoveItem (Variables|Default|GetPlayers_A) _asathena_player_controller)))

(event Custom|Score_A
  (Math|Integer|IncrementInt (Variables|Default|GetScore_Side_A)))

(event Custom|PanelCheck_B
  (params (OverlappedComponent :type Primitive%20Component%20Object%20Reference) (OtherActor :type Actor%20Object%20Reference) (OtherComp :type Primitive%20Component%20Object%20Reference) (OtherBodyIndex :type Integer) (bFromSweep :type Boolean) (SweepResult :type Hit%20Result%20Structure%20%28by%20ref%29))
  ;[0.9093 0.3501 1] Panel Occupancy B
  (if (and (Variables|Default|GetbPanelActive) (Variables|Default|GetbScoreReady))
    (bind _asfort_player_pawn_athena (Utilities|Casting|CastToFortPlayerPawnAthena OtherActor))
    (bind _asathena_player_controller (Utilities|Casting|CastToAthena_PlayerController (Pawn|GetController _asfort_player_pawn_athena)))
    (Utilities|Array|AddUnique (Variables|Default|GetPlayers_B) _asathena_player_controller)))

(event Custom|Score_B
  (Math|Integer|IncrementInt (Variables|Default|GetScore_Side_B)))

(event Custom|ClearScore
  (Variables|Default|SetScore_Side_A 0)
  (Variables|Default|SetScore_Side_B 0))

(event Custom|PanelLeave_B
  (params (OverlappedComponent :type Primitive%20Component%20Object%20Reference) (OtherActor :type Actor%20Object%20Reference) (OtherComp :type Primitive%20Component%20Object%20Reference) (OtherBodyIndex :type Integer))
  ;[0.3168 0 0.4688] Panel Departure B
  (if (and (Variables|Default|GetbPanelActive) (Variables|Default|GetbScoreReady))
    (bind _asfort_player_pawn_athena (Utilities|Casting|CastToFortPlayerPawnAthena OtherActor))
    (bind _asathena_player_controller (Utilities|Casting|CastToAthena_PlayerController (Pawn|GetController _asfort_player_pawn_athena)))
    (Utilities|Array|RemoveItem (Variables|Default|GetPlayers_B) _asathena_player_controller)))

(event Collision|EventActorBeginOverlap
  (params (OtherActor :type Actor%20Object%20Reference)))

(event EventTick
  (params (DeltaSeconds :type Float%20%28single-precision%29)))

(fn HighlightVFXLocation_SideA
  (return (Math|Vector|MakeVector (.x (Math|Transform|InverseTransformLocation (Transformation|GetWorldTransform (Variables|Default|GetDanceFloor_A)) (Transformation|GetWorldLocation (Utilities|Array|Get(acopy) (Variables|Default|GetOverlaps_A) (Variables|Default|GetA_Index))))) (.y (Math|Transform|InverseTransformLocation (Transformation|GetWorldTransform (Variables|Default|GetDanceFloor_A)) (Transformation|GetWorldLocation (Utilities|Array|Get(acopy) (Variables|Default|GetOverlaps_A) (Variables|Default|GetA_Index))))) 6.0)))

(fn HighlightVFXLocation_SideB
  (return (Math|Vector|MakeVector (.x (Math|Transform|InverseTransformLocation (Transformation|GetWorldTransform (Variables|Default|GetDanceFloor_B)) (Transformation|GetWorldLocation (Utilities|Array|Get(acopy) (Variables|Default|GetOverlaps_B) (Variables|Default|GetB_Index))))) (.y (Math|Transform|InverseTransformLocation (Transformation|GetWorldTransform (Variables|Default|GetDanceFloor_B)) (Transformation|GetWorldLocation (Utilities|Array|Get(acopy) (Variables|Default|GetOverlaps_B) (Variables|Default|GetB_Index))))) 6.0)))

(fn OnRep_A_Index)

(fn OnRep_B_Index)

(fn OnRep_bPanelActive
  (if (Variables|Default|GetbPanelActive)
    (CallFunction|PauseMainOverlap)
    (Utilities|Time|ClearTimerbyFunctionName 0 "ClearGame") ; Clear timer that ends game
    (Class|TwinScoreboardINF|Startup(Message) (Variables|Default|GetScoreboard))
    (Audio|Components|Audio|FadeOut (Variables|Default|Getdisco_puzzle_attract_music) 1.0 0.0 "Linear" :!class /Script/Engine.AudioComponent)
    (Utilities|Time|SetTimerbyFunctionName 0 "ScoringLoop" (Variables|Default|GetScoringLoopTime) true)
    (Utilities|Time|ClearTimerbyFunctionName 0 "Attract_A")
    (Utilities|Time|ClearTimerbyFunctionName 0 "Attract_B")
    ;[1 0 0.8711] AUDIO - The game starts here, the delay before RoundTimer is the time between the game starting and the first panel lighting up.
    (Utilities|Time|SetTimerbyFunctionName 0 "RoundTimer" 2.0)
    (Audio|PlaySoundatLocation (Variables|Default|GetDanceOff_Start_Sound) (Transformation|GetWorldLocation (Variables|Default|GetDefaultSceneRoot)))
    (Utilities|Time|SetTimerbyFunctionName 0 "GameMusic" 2.0)
    (else
      (Utilities|Time|SetTimerbyFunctionName 0 "ClearGame" 2.0) ; End game if board isn't occupied
      (Networking|SwitchHasAuthority
        (:Authority
          (CallFunction|PauseMainOverlap))
        (:Remote
          (CallFunction|WinFXEvent))))))

(fn OnRep_bScoreReady)

(fn OnRep_CurrentRound)

(fn OnRep_Panel_A_Occupied)

(fn OnRep_Panel_B_Occupied)

(fn OnRep_Score_Side_A
  (bind _scoringplayers_a (Variables|Default|GetScoringPlayers_A))
  (Networking|SwitchHasAuthority
    (:Remote
      (if (> (Utilities|Array|Length _scoringplayers_a) 0)
        (for _array_element _scoringplayers_a
          (Utilities|IsValid _array_element
            (:Is%20Valid
              (Utilities|IsValid (Game|Player|GetPlayerPawn _array_element)
                (:Is%20Valid
                  (GameplayCue|ExecuteGameplayCue%28Burst%29Local (Game|Player|GetPlayerPawn _array_element) (Variables|Default|GetDanceOffPointTag) (Ability|GameplayCue|MakeGameplayCueParameters :NormalizedMagnitude 1.0 :RawMagnitude 1.0 :Location "0,0,0" :Normal "0,0,0" :Instigator self :GameplayEffectLevel 1 :AbilityLevel 1 :bReplicateLocationWhenUsingMinimalRepProxy false)))))))))))

(fn OnRep_Score_Side_B
  (bind _scoringplayers_b (Variables|Default|GetScoringPlayers_B))
  (Networking|SwitchHasAuthority
    (:Remote
      (if (> (Utilities|Array|Length _scoringplayers_b) 0)
        (for _array_element _scoringplayers_b
          (Utilities|IsValid _array_element
            (:Is%20Valid
              (Utilities|IsValid (Game|Player|GetPlayerPawn _array_element)
                (:Is%20Valid
                  (GameplayCue|ExecuteGameplayCue%28Burst%29Local (Game|Player|GetPlayerPawn _array_element) (Variables|Default|GetDanceOffPointTag) (Ability|GameplayCue|MakeGameplayCueParameters :NormalizedMagnitude 0.0 :RawMagnitude 1.0 :Location "0,0,0" :Normal "0,0,0" :Instigator self :GameplayEffectLevel 1 :AbilityLevel 1 :bReplicateLocationWhenUsingMinimalRepProxy false)))))))))))

(fn OnRep_Side_A_Occupied)

(fn OnRep_Side_B_Occupied)

(fn OnRep_SideA_Quan
  (bind _disco_puzzle_attract_music (Variables|Default|Getdisco_puzzle_attract_music))
  (if (Variables|Default|GetbPanelActive)
    _
    (elif (>= (Variables|Default|GetSideA_Quan) 1)
      (if (>= (Variables|Default|GetSideB_Quan) 1)
        (Variables|Default|SetbPanelActive true)
        (Audio|Components|Audio|FadeOut (Variables|Default|Getdisco_puzzle_attract_music) 1.0 0.0 "Linear" :!class /Script/Engine.AudioComponent)
        (else
          (Utilities|Time|SetTimerbyFunctionName 0 "Attract_A" 0.2)
          (if (not (Audio|Components|Audio|IsPlaying _disco_puzzle_attract_music))
            (Audio|Components|Audio|Play _disco_puzzle_attract_music 0.0))))
      (else
        (Utilities|Time|ClearTimerbyFunctionName 0 "Attract_A")
        (Audio|Components|Audio|FadeOut (Variables|Default|Getdisco_puzzle_attract_music) 1.0 0.0 "Linear" :!class /Script/Engine.AudioComponent)))))

(fn OnRep_SideB_Quan
  (bind _disco_puzzle_attract_music (Variables|Default|Getdisco_puzzle_attract_music))
  (if (Variables|Default|GetbPanelActive)
    _
    (elif (>= (Variables|Default|GetSideB_Quan) 1)
      (if (>= (Variables|Default|GetSideA_Quan) 1)
        (Variables|Default|SetbPanelActive true)
        (Audio|Components|Audio|FadeOut (Variables|Default|Getdisco_puzzle_attract_music) 1.0 0.0 "Linear" :!class /Script/Engine.AudioComponent)
        (else
          (Utilities|Time|SetTimerbyFunctionName 0 "Attract_B" 0.2)
          (if (not (Audio|Components|Audio|IsPlaying _disco_puzzle_attract_music))
            (Audio|Components|Audio|Play _disco_puzzle_attract_music 0.0))))
      (else
        (Utilities|Time|ClearTimerbyFunctionName 0 "Attract_B")
        (Audio|Components|Audio|FadeOut (Variables|Default|Getdisco_puzzle_attract_music) 1.0 0.0 "Linear" :!class /Script/Engine.AudioComponent)))))

(fn ConstructionScript
  (bind _dancefloor_b (Variables|Default|GetDanceFloor_B))
  (bind _dancefloor_a (Variables|Default|GetDanceFloor_A))
  ; FloorBuilding
  (Utilities|FlowControl|Sequence
    (:then_0
      (bind _returnvalue (Rendering|Material|CreateDynamicMaterialInstance _dancefloor_a 3 0 "None" :!class /Script/Engine.PrimitiveComponent))
      (bind _output_get (Variables|Default|SetPanel_A1 _returnvalue)))
    (:then_1
      (bind _returnvalue_2 (Rendering|Material|CreateDynamicMaterialInstance _dancefloor_a 6 0 "None" :!class /Script/Engine.PrimitiveComponent))
      (bind _output_get_2 (Variables|Default|SetPanel_A2 _returnvalue_2)))
    (:then_2
      (bind _returnvalue_3 (Rendering|Material|CreateDynamicMaterialInstance _dancefloor_a 1 0 "None" :!class /Script/Engine.PrimitiveComponent))
      (bind _output_get_3 (Variables|Default|SetPanel_A3 _returnvalue_3)))
    (:then_3
      (bind _returnvalue_4 (Rendering|Material|CreateDynamicMaterialInstance _dancefloor_a 4 0 "None" :!class /Script/Engine.PrimitiveComponent))
      (bind _output_get_4 (Variables|Default|SetPanel_A4 _returnvalue_4)))
    (:then_4
      (bind _returnvalue_5 (Rendering|Material|CreateDynamicMaterialInstance _dancefloor_a 7 0 "None" :!class /Script/Engine.PrimitiveComponent))
      (bind _output_get_5 (Variables|Default|SetPanel_A5 _returnvalue_5)))
    (:then_5
      (bind _returnvalue_6 (Rendering|Material|CreateDynamicMaterialInstance _dancefloor_a 10 0 "None" :!class /Script/Engine.PrimitiveComponent))
      (bind _output_get_6 (Variables|Default|SetPanel_A6 _returnvalue_6)))
    (:then_6
      (bind _returnvalue_7 (Rendering|Material|CreateDynamicMaterialInstance _dancefloor_a 5 0 "None" :!class /Script/Engine.PrimitiveComponent))
      (bind _output_get_7 (Variables|Default|SetPanel_A7 _returnvalue_7)))
    (:then_7
      (bind _returnvalue_8 (Rendering|Material|CreateDynamicMaterialInstance _dancefloor_a 8 0 "None" :!class /Script/Engine.PrimitiveComponent))
      (bind _output_get_8 (Variables|Default|SetPanel_A8 _returnvalue_8)))
    (:then_8
      (bind _returnvalue_9 (Rendering|Material|CreateDynamicMaterialInstance _dancefloor_a 9 0 "None" :!class /Script/Engine.PrimitiveComponent))
      (bind _output_get_9 (Variables|Default|SetPanel_A9 _returnvalue_9))
      (Variables|Default|SetFloors_A (Utilities|Array|MakeArray _output_get _output_get_2 _output_get_3 _output_get_4 _output_get_5 _output_get_6 _output_get_7 _output_get_8 _output_get_9)))
    (:then_9
      (Utilities|FlowControl|Sequence
        (:then_0
          (bind _returnvalue_10 (Rendering|Material|CreateDynamicMaterialInstance _dancefloor_b 3 0 "None" :!class /Script/Engine.PrimitiveComponent))
          (bind _output_get_10 (Variables|Default|SetPanel_B1 _returnvalue_10)))
        (:then_1
          (bind _returnvalue_11 (Rendering|Material|CreateDynamicMaterialInstance _dancefloor_b 6 0 "None" :!class /Script/Engine.PrimitiveComponent))
          (bind _output_get_11 (Variables|Default|SetPanel_B2 _returnvalue_11)))
        (:then_2
          (bind _returnvalue_12 (Rendering|Material|CreateDynamicMaterialInstance _dancefloor_b 1 0 "None" :!class /Script/Engine.PrimitiveComponent))
          (bind _output_get_12 (Variables|Default|SetPanel_B3 _returnvalue_12)))
        (:then_3
          (bind _returnvalue_13 (Rendering|Material|CreateDynamicMaterialInstance _dancefloor_b 4 0 "None" :!class /Script/Engine.PrimitiveComponent))
          (bind _output_get_13 (Variables|Default|SetPanel_B4 _returnvalue_13)))
        (:then_4
          (bind _returnvalue_14 (Rendering|Material|CreateDynamicMaterialInstance _dancefloor_b 7 0 "None" :!class /Script/Engine.PrimitiveComponent))
          (bind _output_get_14 (Variables|Default|SetPanel_B5 _returnvalue_14)))
        (:then_5
          (bind _returnvalue_15 (Rendering|Material|CreateDynamicMaterialInstance _dancefloor_b 10 0 "None" :!class /Script/Engine.PrimitiveComponent))
          (bind _output_get_15 (Variables|Default|SetPanel_B6 _returnvalue_15)))
        (:then_6
          (bind _returnvalue_16 (Rendering|Material|CreateDynamicMaterialInstance _dancefloor_b 5 0 "None" :!class /Script/Engine.PrimitiveComponent))
          (bind _output_get_16 (Variables|Default|SetPanel_B7 _returnvalue_16)))
        (:then_7
          (bind _returnvalue_17 (Rendering|Material|CreateDynamicMaterialInstance _dancefloor_b 8 0 "None" :!class /Script/Engine.PrimitiveComponent))
          (bind _output_get_17 (Variables|Default|SetPanel_B8 _returnvalue_17)))
        (:then_8
          (bind _returnvalue_18 (Rendering|Material|CreateDynamicMaterialInstance _dancefloor_b 9 0 "None" :!class /Script/Engine.PrimitiveComponent))
          (bind _output_get_18 (Variables|Default|SetPanel_B9 _returnvalue_18))
          (Variables|Default|SetFloors_B (Utilities|Array|MakeArray _output_get_10 _output_get_11 _output_get_12 _output_get_13 _output_get_14 _output_get_15 _output_get_16 _output_get_17 _output_get_18)))
        (:then_9
          _))))
  (Variables|Default|SetOverlaps_A (Utilities|Array|MakeArray (Variables|Default|GetOverlapA1) (Variables|Default|GetOverlapA2) (Variables|Default|GetOverlapA3) (Variables|Default|GetOverlapA4) (Variables|Default|GetOverlapA5) (Variables|Default|GetOverlapA6) (Variables|Default|GetOverlapA7) (Variables|Default|GetOverlapA8) (Variables|Default|GetOverlapA9)))
  (Variables|Default|SetOverlaps_B (Utilities|Array|MakeArray (Variables|Default|GetOverlapB1) (Variables|Default|GetOverlapB2) (Variables|Default|GetOverlapB3) (Variables|Default|GetOverlapB4) (Variables|Default|GetOverlapB5) (Variables|Default|GetOverlapB6) (Variables|Default|GetOverlapB7) (Variables|Default|GetOverlapB8) (Variables|Default|GetOverlapB9)))
  (for _array_element (Variables|Default|GetFloors_A)
    (Rendering|Material|SetScalarParameterValue _array_element "Value multiplier" 0.0 :!class /Script/Engine.MaterialInstanceDynamic))
  (for _array_element_2 (Variables|Default|GetFloors_B)
    (Rendering|Material|SetScalarParameterValue _array_element_2 "Value multiplier" 0.0 :!class /Script/Engine.MaterialInstanceDynamic)))
