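; Dance floor panel actor for the DanceOff quest interactable that lights up in random colors.
; Construction creates dynamic materials for the panel and its glow, with the glow vertex offset lowered.
; ActivatePanel picks a random hue and uses a timeline to raise glow state and vertex offset.
; Blink flashes after a random delay and DeactivatePanel resets the materials, both only on non-authority clients; ActivatePanel has no authority check.
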(event Custom|Blink
  ;[1 1 1] Client side fast blink
  (Networking|SwitchHasAuthority
    (:Remote
      (bind _output_get (Variables|Default|SetGlowColor (Math|Color|HSVtoRGB (Math|Random|RandomFloatinRange 0.0 3.6e+02) 1.0 1.0)))
      (Rendering|Material|SetVectorParameterValue (Variables|Default|GetPanelGlow) "GlowColor" _output_get :!class /Script/Engine.MaterialInstanceDynamic)
      (Utilities|FlowControl|Delay (Math|Random|RandomFloatinRange 0.0 0.3))
      (Rendering|Material|SetScalarParameterValue (Variables|Default|GetPanelMaterial) "GlowState" 0.75 :!class /Script/Engine.MaterialInstanceDynamic)
      (Rendering|Material|SetScalarParameterValue (Variables|Default|GetPanelGlow) "VertOffset" 3e+01 :!class /Script/Engine.MaterialInstanceDynamic)
      (Utilities|FlowControl|Delay 0.2)
      (Rendering|Material|SetScalarParameterValue (Variables|Default|GetPanelMaterial) "GlowState" 0.0 :!class /Script/Engine.MaterialInstanceDynamic)
      (Rendering|Material|SetScalarParameterValue (Variables|Default|GetPanelGlow) "VertOffset" 0.0 :!class /Script/Engine.MaterialInstanceDynamic))))

(event Custom|ActivatePanel
  (bind _panelglow (Variables|Default|GetPanelGlow))
  ;[1 1 1] Turn on panel (client side only)
  (bind _output_get (Variables|Default|SetGlowColor (Math|Color|HSVtoRGB (Math|Random|RandomFloatinRange 0.0 3.6e+02) 1.0 1.0)))
  (Rendering|Material|SetVectorParameterValue (Variables|Default|GetPanelGlow) "GlowColor" _output_get :!class /Script/Engine.MaterialInstanceDynamic)
  (bind (_direction _newtrack_0) (Timeline|Timeline_0 PlayFromStart)
    (:Update
      (Rendering|Material|SetScalarParameterValue (Variables|Default|GetPanelMaterial) "GlowState" _newtrack_0 :!class /Script/Engine.MaterialInstanceDynamic)
      (Rendering|Material|SetScalarParameterValue _panelglow "GlowState" _newtrack_0 :!class /Script/Engine.MaterialInstanceDynamic)
      (Rendering|Material|SetScalarParameterValue _panelglow "VertOffset" (Math|Float|Lerp -32.0 0.0 _newtrack_0) :!class /Script/Engine.MaterialInstanceDynamic))))

(event Custom|DeactivatePanel
  ;[1 1 1] Shutdown
  (Networking|SwitchHasAuthority
    (:Remote
      (Rendering|Material|SetScalarParameterValue (Variables|Default|GetPanelMaterial) "GlowState" 0.0 :!class /Script/Engine.MaterialInstanceDynamic)
      (Rendering|Material|SetScalarParameterValue (Variables|Default|GetPanelGlow) "VertOffset" 0.0 :!class /Script/Engine.MaterialInstanceDynamic))))

(event Collision|EventActorBeginOverlap
  (params (OtherActor :type Actor%20Object%20Reference)))

(event EventTick
  (params (DeltaSeconds :type Float%20%28single-precision%29)))

(fn ConstructionScript
  (bind _returnvalue (Rendering|Material|CreateDynamicMaterialInstance (Variables|Default|GetSM_DanceFloor_Panel_01) 0 0 "None" :!class /Script/Engine.PrimitiveComponent))
  (Variables|Default|SetPanelMaterial _returnvalue)
  (bind _returnvalue_2 (Rendering|Material|CreateDynamicMaterialInstance (Variables|Default|GetSM_DanceFloor_Panel_01) 1 0 "None" :!class /Script/Engine.PrimitiveComponent))
  (bind _output_get (Variables|Default|SetPanelGlow _returnvalue_2))
  (Rendering|Material|SetScalarParameterValue _output_get "VertOffset" -32.0 :!class /Script/Engine.MaterialInstanceDynamic))
