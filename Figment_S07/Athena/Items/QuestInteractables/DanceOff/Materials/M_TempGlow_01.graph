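# Default Lit material that provides controllable glow for the Dance Off quest interactable, previewed on the dance floor frame mesh.
# GlowColor scaled by GlowState forms the base glow.
# A DanceFloor static switch adds masked FrameGlow color, scaled by DanceFrame, with FrameState blending from a sine pulse to steady full brightness.
# A separate branch modulates the base glow by an offset camera vector in tangent space for a view-dependent falloff.
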Material /Figment_S07/Athena/Items/QuestInteractables/DanceOff/Materials/M_TempGlow_01
graph {
 properties {
  pixelDepthOffsetMode=PDOM_Legacy
  previewMesh=/Figment_S07/Athena/Items/QuestInteractables/DanceOff/Meshes/SM_DanceFloor_Frame_01
  usedShadingModels=Default Lit
 }
 nodes [
  n1 = Multiply(A=ScalarParameter() {parameterName=GlowState}, B=VectorParameter() {parameterName=GlowColor, defaultValue=<rgba 0.9896 0 0.8781 0>}.RGB)
  n2 = Multiply(A=TextureSample() {samplerSource=SSM_Wrap_WorldGroupSettings, texture=/Figment_S07/Athena/Items/QuestInteractables/DanceOff/Textures/T_DanceFloor_Mask}.R)
  n3 = ConstantBiasScale(None=Sine(None=Time()) {period=2})
  n4 = Multiply(
   A=Multiply(
    A=VectorParameter() {parameterName=FrameGlow, defaultValue=<rgba 1 0.2923 0.8512 0>}.RGB
    B=ScalarParameter() {defaultValue=1, parameterName=DanceFrame}
   )
   B=LinearInterpolate(A=LinearInterpolate(Alpha=@n3) {constA=0.05}, B=Constant() {r=1}, Alpha=ScalarParameter() {parameterName=FrameState})
  )
  Add(A=@n1, B=StaticSwitchParameter(True=Multiply(A=@n2, B=@n4), False=Constant()) {parameterName=DanceFloor})
  Constant() {r=0.4}
  Constant() {r=1}
  n5 = Add(A=CameraVectorWS(), B=AppendVector(A=Constant2Vector(), B=ScalarParameter() {defaultValue=1.6, parameterName=Offset}))
  n6 = Clamp(None=ComponentMask(None=Transform(None=@n5) {transformSourceType=TRANSFORMSOURCE_World, transformType=TRANSFORM_Tangent}) {b=true})
  LinearInterpolate(A=Constant() {r=0.05}, B=Multiply(A=@n6, B=@n1), Alpha=@n6)
 ]
}
