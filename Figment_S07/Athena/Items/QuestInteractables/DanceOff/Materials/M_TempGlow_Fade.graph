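# Masked, unlit, two-sided glow material for the DanceOff quest interactable, with a GlowColor tint scaled by GlowState.
# A vertical UV gradient, camera distance and an upward-panning or scrolling tiled texture multiply with GlowState into a temporal dither threshold, thinning the glow as the camera nears.
# The graph also holds a view-angle color term and a UV-split GlowOffset vector, with no material output connections shown.
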Material /Figment_S07/Athena/Items/QuestInteractables/DanceOff/Materials/M_TempGlow_Fade
graph {
 properties {blendMode=BLEND_Masked, pixelDepthOffsetMode=PDOM_Legacy, shadingModel=MSM_Unlit, twoSided=true, usedShadingModels=Unlit}
 nodes [
  n1 = ScalarParameter() {defaultValue=1, parameterName=GlowState}
  n2 = MaterialFunctionCall(Vector 3=Subtract(A=WorldPosition().XYZ, B=CameraPositionWS())) {materialFunction=/Engine/Functions/Engine_MaterialFunctions02/Utility/VectorLength}
  n3 = ComponentMask(None=TextureCoordinate()) {g=true}
  n4 = Multiply(A=@n1, B=Multiply(A=Divide(A=@n2.V3 Length) {constB=384}, B=@n3))
  n5 = TextureSample(UVs=Panner(Coordinate=TextureCoordinate() {uTiling=5}) {speedY=0.3}) {samplerSource=SSM_Wrap_WorldGroupSettings, texture=/Figment_S07/Athena/Items/QuestInteractables/DanceOff/Textures/T_Smooth_tile}
  MaterialFunctionCall(Alpha Threshold=Clamp(None=Multiply(A=@n4, B=@n5.R))) {materialFunction=&DitherTemporalAA_Fortnite}
  n6 = VectorParameter() {parameterName=GlowColor, defaultValue=<rgba 0.9896 0 0.8781 0>}
  Multiply(A=@n1, B=@n6.RGB)
  n7 = Transform(None=Add(A=CameraVectorWS(), B=Constant3Vector() {constant=<rgba 0 0 1.0628 0>})) {transformSourceType=TRANSFORMSOURCE_World, transformType=TRANSFORM_Tangent}
  Multiply(A=Clamp(None=ComponentMask(None=@n7) {b=true}), B=@n6.RGB)
  n8 = TextureCoordinate()
  n9 = Ceil(None=Add(A=ComponentMask(None=@n8) {g=true}) {constB=-0.5})
  n10 = AppendVector(A=Constant2Vector(), B=ScalarParameter() {parameterName=GlowOffset})
  Multiply(A=Clamp(None=Subtract(B=@n9)), B=@n10)
  Subtract(B=@n3)
  Transform(None=@n10)
 ]
}
